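# Translucent unlit, two-sided material for the LoneWolf skydiving contrail trail on Niagara mesh particles.
# Scrolling smoke and noise textures drive color through a curve atlas row, tinted by particle color and time of day.
# Opacity combines panned masks, a second curve row, depth, view-angle and near-camera fades.
# World position offset layers sine waves, normal-based noise, optional spline thickening and a clamped pull along the trail, steered by dynamic parameters.
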Material /BRCosmetics/Athena/Effects/Particles/Skydiving/NS_Contrail_LoneWolf/M_Contrail_LoneWolf_Trail
graph {
 properties {
  bUsedWithNiagaraMeshParticles=true
  blendMode=BLEND_Translucent
  pixelDepthOffsetMode=PDOM_Legacy
  shadingModel=MSM_Unlit
  twoSided=true
  usedShadingModels=Unlit
 }
 nodes [
  n1 = DynamicParameter() {parameterIndex=2, paramNames [X,Y,Z,WPOIntensity]}
  n2 = MaterialFunctionCall(X=@n1.X, Y=@n1.X, Z=@n1.X) {materialFunction=&MakeFloat3}
  n3 = ParticlePositionWS()
  n4 = DynamicParameter() {parameterIndex=3, paramNames [WPO_Strength,Max,Param3,Param4]}
  n5 = TextureCoordinate()
  n6 = TextureCoordinate() {coordinateIndex=1}
  n7 = ScalarParameter() {group=UV, parameterName="UV_0,1"}
  n8 = OneMinus(None=ComponentMask(None=LinearInterpolate(A=@n5, B=@n6, Alpha=@n7)) {r=true})
  n9 = ScalarParameter() {defaultValue=2, group=WPO, parameterName=InterpolatePower}
  n10 = LinearInterpolate(
   B=Clamp(None=Subtract(A=@n2.Result, B=@n3), Min=Multiply(A=@n4.Max) {constA=-1, constB=-1}, Max=@n4.Max) {maxDefault=50, minDefault=-50}
   Alpha=Multiply(A=@n1.WPOIntensity, B=Power(Base=@n8, Exp=@n9) {constExponent=3})
  )
  n11 = ScalarParameter() {defaultValue=1, group=WPO, parameterName=Interpolate}
  n12 = LinearInterpolate(A=TextureCoordinate(), B=TextureCoordinate() {coordinateIndex=1}, Alpha=ScalarParameter() {group=UV, parameterName="UV_0,1"})
  n13 = ComponentMask(None=@n12) {r=true}
  n14 = DynamicParameter() {paramNames [Spline_Width, UV_Speed, WPO_Wave_Size, UV_Tiling]}
  n15 = Time()
  n16 = DynamicParameter() {parameterIndex=1, paramNames [WPO_Wave_Speed, UV_Speed, WPO_Wave_Size, UV_Tiling]}
  n17 = Sine(None=Add(A=Multiply(A=@n13, B=@n14.WPO_Wave_Size) {constB=1.5}, B=Multiply(A=@n15, B=@n16.WPO_Wave_Speed)))
  n18 = Constant3Vector() {constant=<rgba 1 0 0 0>}
  n19 = Multiply(
   A=Transform(None=Multiply(A=@n17, B=@n18)) {transformSourceType=TRANSFORMSOURCE_View}
   B=Multiply(
    A=ScalarParameter() {defaultValue=14, group=WPO, parameterName=WPO_Wave}
    B=DynamicParameter() {parameterIndex=3, paramNames [WPO_Strength,Max,Param3,Param4]}.WPO_Strength
   )
  ) {constB=10}
  n20 = Add(
   A=LinearInterpolate(
    A=TextureCoordinate() {uTiling=0.1}
    B=TextureCoordinate() {coordinateIndex=1, uTiling=0.1}
    Alpha=ScalarParameter() {group=UV, parameterName="UV_0,1"}
   )
   B=Time()
  )
  n21 = Multiply(
   A=Multiply(A=TextureSample(UVs=@n20) {texture=/Game/Effects/Fort_Effects/Textures/Power_Source/T_Power_Source_Sphere_Texture}.RGB, B=VertexNormalWS())
   B=ScalarParameter() {defaultValue=10, group=WPO, parameterName=WPO_Noise}
  ) {constB=12}
  n22 = StaticComponentMaskParameter(None=Subtract(A=WorldPosition().XYZ, B=ObjectPositionWS())) {defaultB=true}
  n23 = DynamicParameter() {paramNames [Spline_Width, UV_Speed, WPO_Wave_Size, UV_Tiling]}
  n24 = Multiply(A=ScalarParameter() {group=SplineThicken, parameterName=Spline_Base}, B=@n23.Spline_Width)
  n25 = Multiply(A=ScalarParameter() {group=SplineThicken, parameterName=Spline_Tip}, B=@n23.Spline_Width)
  n26 = Add(
   A=Multiply(A=Add(A=@n19, B=@n21), B=Divide(A=@n22) {constB=200})
   B=LinearInterpolate(
    B=MaterialFunctionCall(WidthBase=@n24, WidthTip=@n25) {materialFunction=&SplineThicken}.WorldPositionOffset
    Alpha=ScalarParameter() {group=SplineThicken, parameterName=SplineThicken}
   )
  )
  LinearInterpolate(A=Add(A=Multiply(A=@n10, B=@n11), B=@n26), B=@n26, Alpha=ScalarParameter() {group=WPO, parameterName=WPO_Offset})
  n27 = TextureCoordinate()
  n28 = Power(Base=Multiply(A=MaterialFunctionCall(Float2=@n27) {materialFunction=&BreakOutFloat2Components}.R) {constB=0.7}) {constExponent=3}
  n29 = VectorParameter() {group=UV, parameterName=UVTiling, defaultValue=<rgba 0.45 1.7 0.15 0.5>}
  n30 = MaterialFunctionCall(Float2=Multiply(A=@n12, B=AppendVector(A=@n29.R, B=@n29.R))) {materialFunction=&BreakOutFloat2Components}
  n31 = DynamicParameter() {paramNames [Spline_Width, UV_Speed, WPO_Wave_Size, UV_Tiling]}
  n32 = Multiply(A=Time(), B=DynamicParameter() {paramNames [Spline_Width, UV_Speed, WPO_Wave_Size, UV_Tiling]}.UV_Speed)
  n33 = Panner(Coordinate=AppendVector(A=Multiply(A=@n30.R, B=@n31.UV_Tiling), B=@n30.G), Time=@n32) {speedX=1.4}
  n34 = Panner(
   Coordinate=LinearInterpolate(
    A=TextureCoordinate() {uTiling=0.36, vTiling=1.4}
    B=TextureCoordinate() {coordinateIndex=1, uTiling=0.4}
    Alpha=ScalarParameter() {group=UV, parameterName="UV_0,1"}
   )
   Time=@n32
  ) {speedX=1}
  n35 = LinearInterpolate(
   A=Multiply(A=TextureSample(UVs=@n33) {texture=&T_Inky_Smoke_Tile}.R) {constB=1.8}
   B=Power(Base=TextureSample(UVs=@n34) {samplerType=SAMPLERTYPE_Masks, texture=/Game/Weapons/WeaponSkins/Wraps/Textures/Mash/Mash_Noise}.R) {constExponent=0.8}
   Alpha=Constant() {r=0.5}
  )
  n36 = Clamp(None=LinearInterpolate(A=Add(A=@n28, B=@n35), B=@n35, Alpha=ScalarParameter() {parameterName=UseMask}))
  n37 = CurveAtlasRowParameter(CurveTime=@n36) {
   atlas=/Game/Athena/LinearColorCurves/VFX_CurveAtlas
   curve=/Game/Athena/LinearColorCurves/Curve_Contrail_LoneWolf_Color
   defaultValue=50
   group=Color
   parameterName=Curve_Color
  }
  n38 = VectorParameter() {group=Color, parameterName=ColorA, defaultValue=<rgba 1 0.0351 0 1>}
  n39 = VectorParameter() {group=Color, parameterName=ColorB, defaultValue=<rgba 0 0 0 1>}
  n40 = Desaturation(
   None=LinearInterpolate(A=@n37, B=LinearInterpolate(A=@n38.RGB, B=@n39.RGB, Alpha=@n36), Alpha=ScalarParameter() {group=Color, parameterName=UseCurveForColor})
   Fraction=ScalarParameter() {defaultValue=-0.6, group=Color, parameterName=Desaturation}
  )
  n41 = MaterialFunctionCall(
   Color_Morning=ScalarParameter() {defaultValue=0.3, parameterName=ColorTimeOfDay_Morning}
   Color_Day=ScalarParameter() {defaultValue=1, parameterName=ColorTimeOfDay_Day}
   Color_Evening=ScalarParameter() {defaultValue=0.4, parameterName=ColorTimeOfDay_Evening}
   Color_Night=ScalarParameter() {defaultValue=0.7, parameterName=ColorTimeOfDay_Night}
  ) {materialFunction=&Color_Time_of_Day}
  n42 = ParticleColor()
  MaterialFunctionCall(Emissive=Multiply(A=Multiply(A=@n40, B=@n41.Result), B=@n42.RGB)) {materialFunction=&MF_FX_EmissiveAdjustments}
  n43 = ScalarParameter() {defaultValue=0.7, parameterName=TimeOfDay_Morning}
  n44 = ScalarParameter() {defaultValue=1, parameterName=TimeOfDay_Day}
  n45 = ScalarParameter() {defaultValue=1, parameterName=TimeOfDay_Evening}
  n46 = ScalarParameter() {defaultValue=0.7, parameterName=TimeOfDay_Night}
  n47 = TextureSample(UVs=Panner(Coordinate=@n12) {speedX=1.4}) {texture=/Game/Global/Textures/Default/Masks_and_Gradients/CylGradient}
  n48 = Panner(Coordinate=Multiply(A=@n12, B=AppendVector(A=@n29.B, B=@n29.B)), Time=@n32) {speedX=0.46}
  n49 = Add(
   A=Multiply(A=TextureSample(UVs=@n48) {samplerType=SAMPLERTYPE_Masks, texture=&Mash_Noise}.R) {constB=0.15}
   B=@n12
  )
  n50 = LinearInterpolate(
   A=Multiply(A=TextureSample(UVs=@n33) {texture=/Game/Effects/Fort_Effects/Textures/Smoke/T_Lobber_Trail_Mask}.R) {constB=0.25}
   B=Add(A=Multiply(A=@n47.R) {constB=6}, B=TextureSample(UVs=@n49) {texture=/Game/Global/Textures/Default/Masks_and_Gradients/gradient_stasis}.R)
   Alpha=Constant() {r=0.5}
  )
  n51 = SphereMask(A=TextureCoordinate() {uTiling=0.2, vTiling=1.2}, B=Constant2Vector() {g=0.5, r=1}) {attenuationRadius=1, hardnessPercent=55}
  n52 = ScalarParameter() {defaultValue=15, parameterName=Param_1}
  n53 = ComponentMask(None=@n12) {r=true}
  n54 = Panner(
   Coordinate=LinearInterpolate(
    A=TextureCoordinate() {uTiling=0.2, vTiling=0.5}
    B=TextureCoordinate() {coordinateIndex=1, uTiling=0.36, vTiling=0.6}
    Alpha=ScalarParameter() {group=UV, parameterName="UV_0,1"}
   )
   Time=Multiply(A=Time(), B=DynamicParameter() {paramNames [Spline_Width, UV_Speed, WPO_Wave_Size, UV_Tiling]}.UV_Speed)
  ) {speedX=0.7, speedY=0.1}
  n55 = Subtract(
   A=@n53
   B=Multiply(A=TextureSample(UVs=@n54) {samplerType=SAMPLERTYPE_Normal, texture=/Game/Weapons/WeaponSkins/Wraps/Textures/Mash/Mash_N}.R) {constB=7}
  )
  n56 = Subtract(A=ComponentMask(None=TextureCoordinate() {uTiling=0.2, vTiling=1.2}) {g=true}) {constB=0.5}
  n57 = Multiply(
   A=LinearInterpolate(A=Multiply(A=@n51, B=@n52) {constB=30}, B=Multiply(A=@n53, B=@n55), Alpha=ScalarParameter() {defaultValue=0.77, parameterName=Param_2})
   B=OneMinus(None=Abs(None=@n56))
  )
  n58 = Multiply(A=Power(Base=@n53)) {constB=2}
  n59 = CurveAtlasRowParameter(CurveTime=Clamp(None=Add(A=@n57, B=@n58))) {
   atlas=&VFX_CurveAtlas
   curve=/Game/Athena/LinearColorCurves/Curve_Contrail_LoneWolf_Mask
   defaultValue=51
   group=Color
   parameterName=Curve_Color
  }
  n60 = Multiply(A=Multiply(A=@n50, B=@n59.R), B=SphereMask(A=ComponentMask(None=@n12) {r=true}, B=Constant()) {attenuationRadius=1, hardnessPercent=88})
  n61 = ScalarParameter() {parameterName=UseMask}
  n62 = ScalarParameter() {defaultValue=0.5, group=Fade, parameterName=FadeBase}
  n63 = DepthFade(
   Opacity=Multiply(
    A=LinearInterpolate(A=@n60, B=@n50, Alpha=@n61)
    B=MaterialFunctionCall("FadeSensitivity (optional)"=@n62) {materialFunction=&FadeBasedOnViewAngle}.Result
   )
   FadeDistance=ScalarParameter() {defaultValue=20, group=Fade, parameterName=DepthFade}
  )
  Multiply(
   A=Multiply(
    A=MaterialFunctionCall(Value_Morning=@n43, Value_Day=@n44, Value_Evening=@n45, Value_Night=@n46) {materialFunction=&Scalar_Time_of_Day}.Result
    B=Multiply(A=@n42.A, B=@n63)
   )
   B=MaterialFunctionCall() {materialFunction=&Near_Camera_Fade}.Result
  )
  ObjectPositionWS()
  Reroute(None=@n50)
 ]
}
