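# Translucent, unlit, two-sided Niagara mesh-particle material for the Lone Wolf skydiving contrail trail.
# Emissive blends a panning-flame two-color gradient with fresnel color via UseEmissiveColor, which also switches opacity to solid.
# Opacity multiplies sine edge mask, fresnel, distorted panning trail texture, particle alpha, view-angle fade and time-of-day scaling.
# UseWPO fades out vertex offset built from a view-space sine wave and clamped, UV-weighted pull toward a dynamic-parameter position.
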Material /BRCosmetics/Athena/Effects/Particles/Skydiving/NS_Contrail_LoneWolf/M_Contrail_LoneWolf_Trail_Fresnel
graph {
 properties {
  bUsedWithNiagaraMeshParticles=true
  blendMode=BLEND_Translucent
  pixelDepthOffsetMode=PDOM_Legacy
  shadingModel=MSM_Unlit
  twoSided=true
  usedShadingModels=Unlit
 }
 nodes [
  n1 = TextureCoordinate()
  n2 = ComponentMask(None=Multiply(A=@n1, B=DynamicParameter() {defaultValue=<rgba 1 0.7 1 1>, paramNames [WPO_Spline_Size, WPO_Wave_Size, Base, Param4]}.WPO_Wave_Size)) {g=true}
  n3 = Multiply(A=Time())
  n4 = Power(Base=Sine(None=Add(A=@n2, B=@n3))) {constExponent=3}
  n5 = Constant3Vector() {constant=<rgba 1 1 0 0>}
  n6 = Multiply(A=Transform(None=Multiply(A=@n4, B=@n5)) {transformSourceType=TRANSFORMSOURCE_View}, B=ScalarParameter() {defaultValue=5, parameterName=WPO})
  n7 = OneMinus(None=ComponentMask(None=@n1) {g=true})
  n8 = DynamicParameter() {parameterIndex=2, paramNames [X,Y,Z,WPOIntensity]}
  n9 = MaterialFunctionCall(X=@n8.X, Y=@n8.X, Z=@n8.X) {materialFunction=&MakeFloat3}
  n10 = ParticlePositionWS()
  n11 = DynamicParameter() {parameterIndex=3, paramNames [WPO_Strength,Max,Param3,Param4]}
  n12 = OneMinus(None=ComponentMask(None=TextureCoordinate()) {g=true})
  n13 = ScalarParameter() {defaultValue=2, group=WPO, parameterName=InterpolatePower}
  n14 = LinearInterpolate(
   B=Clamp(None=Subtract(A=@n9.Result, B=@n10), Min=Multiply(A=@n11.Max) {constA=-1, constB=-1}, Max=@n11.Max) {maxDefault=50, minDefault=-50}
   Alpha=Multiply(A=@n8.WPOIntensity, B=Power(Base=@n12, Exp=@n13) {constExponent=3})
  )
  n15 = ScalarParameter() {defaultValue=1, group=WPO, parameterName=Interpolate}
  LinearInterpolate(A=Add(A=Multiply(A=@n6, B=@n7), B=Multiply(A=@n14, B=@n15)), B=Constant(), Alpha=ScalarParameter() {parameterName=UseWPO})
  n16 = Divide(A=MaterialFunctionCall(Float2=TextureCoordinate()) {materialFunction=&BreakOutFloat2Components}.R)
  n17 = Power(Base=OneMinus(None=Sine(None=@n16)))
  n18 = LinearInterpolate(
   A=VectorParameter() {parameterName=ColorA, defaultValue=<rgba 0.9375 0 0.0069 1>}.RGB
   B=VectorParameter() {parameterName=ColorB, defaultValue=<rgba 4 0.1483 0 1>}.RGB
   Alpha=LinearInterpolate(Alpha=Multiply(A=@n17) {constB=2}) {constA=0.7}
  )
  n19 = ScalarParameter() {defaultValue=0.5, parameterName=Param_3}
  n20 = ScalarParameter() {defaultValue=0.25, parameterName=Param_4}
  n21 = Panner(Coordinate=Multiply(A=TextureCoordinate() {uTiling=0.7, vTiling=0.3}, B=AppendVector(A=@n19, B=@n20))) {speedY=0.35}
  n22 = Add(A=Multiply(A=TextureSample(UVs=@n21) {samplerType=SAMPLERTYPE_LinearGrayscale, texture=/BRCosmetics/Weapons/FORT_Melee/Pickaxe_Cube_Ninja_Male/FX/T_FluidFlameTile}.R) {constB=20})
  n23 = Fresnel() {baseReflectFraction=0.01}
  n24 = VectorParameter() {parameterName=EmissiveColor, defaultValue=<rgba 6 0 0 1>}
  MaterialFunctionCall(Emissive=LinearInterpolate(A=Multiply(A=@n18, B=@n22), B=Multiply(A=@n23, B=@n24.RGB), Alpha=ScalarParameter() {parameterName=UseEmissiveColor})) {materialFunction=&MF_FX_EmissiveAdjustments}
  n25 = ComponentMask(None=TextureCoordinate()) {g=true}
  n26 = Power(Base=Sine(None=Divide(A=@n25)))
  n27 = Fresnel()
  n28 = TextureCoordinate() {uTiling=0.4, vTiling=0.55}
  n29 = TextureCoordinate() {uTiling=0.4, vTiling=0.2}
  n30 = Multiply(A=TextureSample(UVs=Panner(Coordinate=@n29) {speedX=0.3, speedY=0.8}) {samplerType=SAMPLERTYPE_Normal, texture=/Game/Weapons/WeaponSkins/Wraps/Textures/Mash/Mash_N}.R) {constB=0.18}
  n31 = TextureSample(UVs=Panner(
    Coordinate=Add(A=@n28, B=@n30)
    Time=DynamicParameter() {defaultValue=<rgba 1 0.7 1 1>, paramNames [WPO_Spline_Size, WPO_Wave_Size, Base, Param4]}.Param4
   ) {speedY=1}) {
   samplerSource=SSM_Clamp_WorldGroupSettings
   samplerType=SAMPLERTYPE_LinearColor
   texture=/BRCosmetics/Athena/Effects/Particles/Skydiving/NS_Contrail_CubeNinja/T_CubeNinja_Trail
  }
  n32 = ScalarParameter() {defaultValue=1.6}
  n33 = Multiply(A=Multiply(A=Multiply(A=@n26, B=@n27), B=Multiply(A=@n31.B, B=@n32)), B=ParticleColor().A)
  n34 = MaterialFunctionCall("FadeSensitivity (optional)"=ScalarParameter() {defaultValue=3, group=Fade, parameterName=FadeBase}) {materialFunction=&FadeBasedOnViewAngle}
  Multiply(
   A=LinearInterpolate(A=Multiply(A=@n33, B=@n34.Result), B=Constant() {r=1}, Alpha=ScalarParameter() {parameterName=UseEmissiveColor})
   B=MaterialFunctionCall(
    Color_Morning=ScalarParameter() {defaultValue=1, parameterName=TimeOfDay_Morning}
    Color_Day=ScalarParameter() {defaultValue=1, parameterName=TimeOfDay_Day}
    Color_Evening=ScalarParameter() {defaultValue=0.7, parameterName=TimeOfDay_Evening}
    Color_Night=ScalarParameter() {defaultValue=0.7, parameterName=TimeOfDay_Night}
   ) {materialFunction=&Color_Time_of_Day}.Result
  )
 ]
}
