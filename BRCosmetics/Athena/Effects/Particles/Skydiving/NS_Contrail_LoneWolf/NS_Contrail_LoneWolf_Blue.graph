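# Niagara skydiving contrail, the blue LoneWolf variant, dressing the player's skeleton with trail meshes, hand meshes and rings, blue lights and a spine-mounted body mesh.
# Spawn groups index the sampled bones or sockets, pinning persistent burst particles and two looping, fading trail emitters.
# A diving-attributes module produces a Dive value that drives trail scale and material parameters.
# Smoothed particle position also feeds the trail materials, and time of day scales the lights.
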NiagaraSystem /BRCosmetics/Athena/Effects/Particles/Skydiving/NS_Contrail_LoneWolf/NS_Contrail_LoneWolf_Blue
dependencies {
 usedDataInterfaces [
  &NiagaraDataInterfaceArrayFloat3
  &NiagaraDataInterfaceArrayFloat
  &NiagaraDataInterfaceCurve
  &NiagaraDataInterfaceColorCurve
 ]
 usedDynamicInputs [&RandomRangeLinearColor]
 usedModules [&SystemState]
}
emitters {
 Body_Moon {
  emitterUpdateScript [
   &EmitterState(
    Inactive Response=::NewEnumerator1
    Life Cycle Mode=::NewEnumerator1
    Loop Behavior=::NewEnumerator1
    Loop Duration Mode=::NewEnumerator1
    UseLoopDelay=false
   )
   &SpawnBurst_Instantaneous(Use Spawn Probability=false, UseLoopCountLimit=false)
  ]
  particleSpawnScript [
   &InitializeParticle(Color Mode=::NewEnumerator0, Mesh Scale=<6 0 6>, Mesh Scale Mode=::NewEnumerator4, UsePositionOffset=false, Write Material Random=true)
   &InitialMeshOrientation(Mesh Orientation Mode=::NewEnumerator7, Rotation=<0 0.25 0>, Use Rotation Vector=true)
  ]
  particleUpdateScript [
   &ParticleState(Kill Particles When Lifetime Has Elapsed=false, Let Infinitely Lived Particles Die When Emitter Deactivates=true)
   &SkeletalMeshLocation(
    Bone Sampling Mode=::NewEnumerator4
    MaskSpawnGroup=false
    Sampled Orientation Coordinate Space=::Local
    Sampled Position Offset=<0 18 0>
    Skeletal Mesh=DataInterface() {propertyValues {
      DefaultMesh=null
      ExcludeBoneName=root
      OptionalComponentClass=null
      PreviewMesh=&M_Med_Soldier_04_Preview
      SkinningMode=SkinOnTheFly
      SoftSourceActor=null
      SourceMode=Default
      UvSetIndex=0
      WholeMeshLOD=-1
      bExcludeBone=false
      bReadDeformedGeometry=true
      bRequireCurrentFrameData=true
      ComponentTags []
      FilteredBones [spine_03]
      FilteredSockets []
      MeshUserParameter {parameter {name=None, typeDefHandle {registeredTypeIndex=1}, varData []}}
      SamplingRegions []
     }}
   )
  ]
  rendererClasses [&NiagaraMeshRendererProperties]
  renderers [{rendererClass=&NiagaraMeshRendererProperties, rendererIndex=0}]
 }
 Gloden_Trail {
  emitterUpdateScript [
   &EmitterState(
    Inactive Response=::NewEnumerator1
    Life Cycle Mode=::NewEnumerator1
    Loop Delay=&RandomRangeFloat(Minimum=0, Maximum=1)
    Loop Duration=100
    UseLoopDelay=true
   )
   &SpawnRate(SpawnRate=4, Use Spawn Probability=false)
  ]
  particleSpawnScript [
   &FN_SetSpawnGroup(Number of Spawn Groups=2)
   &InitializeParticle(
    Color Maximum=<rgba 1 1 1 3>
    Color Minimum=<rgba 1 1 1 3>
    Color Mode=::NewEnumerator2
    Lifetime Max=1.1
    Lifetime Min=0.7
    Lifetime Mode=::NewEnumerator1
    Mesh Scale=&Multiply_Vector(
     A=&RandomRangeVector(Minimum=<1 1 0.7>, Maximum=<1 1 1.4>)
     B=&SelectVectorFromArray(
      Direct Array Index=&SpawnGroup()
      Vector Selection Array=DataInterface() {propertyValues {FloatData [<0.3 0.3 0.6>, <0.3 0.3 0.6>, <0.3 0.3 0.3>, <0.25 0.25 0.3>]}}
     )
    )
    Mesh Scale Mode=::NewEnumerator4
    UsePositionOffset=false
    Write Material Random=true
   )
   /SetVariables(
    Particles.MeshScale:Vector3f=&SelectVectorFromArray(
     Direct Array Index=&SpawnGroup()
     Vector Selection Array=DataInterface() {propertyValues {FloatData [<0.78 0.78 1.8>, <0.78 0.78 1.4>, <0.3 0.3 0.4>, <0.3 0.3 0.55>]}}
    )
    Particles.SplineMeshScale:Vector3f=&RandomRangeVector(Minimum=<0.2 0.2 0.5>, Maximum=<0.35 0.35 0.72>)
   ) {enabled=false, isSetParameters=true, module=SetVariables_#1}
   &InitialMeshOrientation(
    Orientation Axis=<0 0 1>
    Orientation Vector=<0 0 1>
    Rotation Coordinate Space=::NewEnumerator0
    Use Orientation Vector=true
    Use Rotation Vector=true
   )
   &CameraOffset(Camera Offset Amount=40)
   &DynamicMaterialParameters(
    Index 0 Param 1=&RandomRangeFloat(Minimum=1, Maximum=2)
    Index 0 Param 2=&RandomRangeFloat(Minimum=1, Maximum=1.8)
    Index 0 Param 3=&RandomRangeFloat(Minimum=10, Maximum=18)
    Module.Param0WriteEnabled=true
    Module.Param1WriteEnabled=true
    Module.Param2WriteEnabled=true
    Module.Param3WriteEnabled=true
   )
  ]
  particleUpdateScript [
   &ParticleState()
   /Game/Effects/Niagara/Modules/Skydiving/FN_DivingAttributes/FN_DivingAttributes_v2(Accumulate Dive=0.1, Accumulate Rest=-0.1, Bone Index=0, SkeletalMesh:NiagaraDataInterfaceSkeletalMesh=$User.Skeletal Mesh01)
   &InterpolateOverTime(Rate Of Change=10, Target Vector Value:NiagaraPosition=$Particles.Position, Variable Type=::NewEnumerator2)
   &SkeletalMeshLocation(
    Bone / Socket Index=$Particles.SpawnGroup
    Bone Sampling Mode=::NewEnumerator4
    MaskSpawnGroup=false
    Orientation Sampling=::NewEnumerator1
    Sampled Orientation Coordinate Space=::Local
    Skeletal Mesh=$User.SK_Mesh
   )
   &ScaleColor(
    Scale Alpha=&FloatFromCurve(
     FloatCurve=DataInterface() {propertyValues {Curve {keys [
         {arriveTangent=1.4675, interpMode=RCIM_Cubic, leaveTangent=1.4675, tangentMode=RCTM_User, time=0, value=0}
         {arriveTangent=-0.8619, interpMode=RCIM_Cubic, leaveTangent=-0.8619, tangentMode=RCTM_User, time=0.4, value=1}
         {arriveTangent=0.8846, interpMode=RCIM_Cubic, leaveTangent=0.8846, tangentMode=RCTM_User, time=1, value=0}
        ]}}}
     Scale Curve=2
    )
    ScaleA=true
    ScaleRGB=true
   )
   /SetVariables(Particles.DynamicMaterialParameter2:Vector4f=&MakeFloat4FromVectorAndFloat("Vector (RGB)"=$Output.InterpolateOverTime.Moving Average, "Float (Alpha)"=1)) {isSetParameters=true}
   &DynamicMaterialParameters(
    Index 0 Param 4=&FloatFromCurve(FloatCurve=DataInterface() {propertyValues {Curve {keys [
         {arriveTangent=1.0462, interpMode=RCIM_Cubic, leaveTangent=1.0462, tangentMode=RCTM_User, time=0, value=-0.7}
         {arriveTangent=0.5091, interpMode=RCIM_Cubic, leaveTangent=0.5091, tangentMode=RCTM_User, time=1, value=1}
        ]}}})
    Index 3 Param 2=&Lerp_Float(A=0, B=100, Alpha=$Particles.Dive)
    Module.Param0WriteEnabled=false
    Module.Param0WriteEnabled3=true
    Module.Param1WriteEnabled=false
    Module.Param1WriteEnabled3=true
    Module.Param2WriteEnabled=false
    Module.Param2WriteEnabled3=true
    Module.Param3WriteEnabled=true
    Module.Param3WriteEnabled3=true
    Write Parameter Index 3=true
   ) {module=DynamicMaterialParameters001}
   &ScaleMeshSize(Scale Factor=&Multiply_VectorByFloat(Vector=<1 1 1>, Float=&Lerp_Float(A=1, B=1.45, Alpha=$Particles.Dive)))
  ]
  rendererClasses [&NiagaraMeshRendererProperties]
  renderers [{rendererClass=&NiagaraMeshRendererProperties, rendererIndex=0}]
 }
 Gloden_Trail001 {
  emitterUpdateScript [
   &EmitterState(
    Inactive Response=::NewEnumerator1
    Life Cycle Mode=::NewEnumerator1
    Loop Delay=&RandomRangeFloat(Minimum=0.5, Maximum=1)
    Loop Duration=100
    UseLoopDelay=true
   )
   &SpawnRate(SpawnRate=3, Use Spawn Probability=false)
  ]
  particleSpawnScript [
   &FN_SetSpawnGroup(Number of Spawn Groups=2)
   &InitializeParticle(
    Color Maximum=<rgba 1 1 1 3>
    Color Minimum=<rgba 1 1 1 3>
    Color Mode=::NewEnumerator2
    Lifetime Max=1
    Lifetime Min=0.7
    Lifetime Mode=::NewEnumerator1
    Mesh Scale=&Multiply_Vector(
     A=&RandomRangeVector(Minimum=<1 1 0.9>, Maximum=<1 1 1.2>)
     B=&SelectVectorFromArray(
      Direct Array Index=&SpawnGroup()
      Vector Selection Array=DataInterface() {propertyValues {FloatData [<0.3 0.3 0.6>, <0.3 0.3 0.6>, <0.3 0.3 0.3>, <0.25 0.25 0.3>]}}
     )
    )
    Mesh Scale Mode=::NewEnumerator4
    UsePositionOffset=false
    Write Material Random=true
   )
   /SetVariables(
    Particles.MeshScale:Vector3f=&SelectVectorFromArray(
     Direct Array Index=&SpawnGroup()
     Vector Selection Array=DataInterface() {propertyValues {FloatData [<0.78 0.78 1.8>, <0.78 0.78 1.4>, <0.3 0.3 0.4>, <0.3 0.3 0.55>]}}
    )
    Particles.SplineMeshScale:Vector3f=&RandomRangeVector(Minimum=<0.2 0.2 0.5>, Maximum=<0.35 0.35 0.72>)
   ) {isSetParameters=true, module=SetVariables_#1}
   &InitialMeshOrientation(
    Orientation Axis=<0 0 1>
    Orientation Vector=<0 0 1>
    Rotation Coordinate Space=::NewEnumerator0
    Use Orientation Vector=true
    Use Rotation Vector=true
   )
   &CameraOffset(Camera Offset Amount=40)
   &DynamicMaterialParameters(
    Index 0 Param 1=&RandomRangeFloat(Minimum=1, Maximum=2)
    Index 0 Param 2=&RandomRangeFloat(Minimum=1, Maximum=1.8)
    Index 0 Param 3=&RandomRangeFloat(Minimum=10, Maximum=18)
    Module.Param0WriteEnabled=true
    Module.Param1WriteEnabled=true
    Module.Param2WriteEnabled=true
    Module.Param3WriteEnabled=true
   )
  ]
  particleUpdateScript [
   &ParticleState()
   &FN_DivingAttributes_v2(Accumulate Dive=0.1, Accumulate Rest=-0.1, Bone Index=0, SkeletalMesh:NiagaraDataInterfaceSkeletalMesh=$User.Skeletal Mesh01)
   &InterpolateOverTime(Rate Of Change=10, Target Vector Value:NiagaraPosition=$Particles.Position, Variable Type=::NewEnumerator2)
   /SetVariables(Particles.DynamicMaterialParameter2:Vector4f=&MakeFloat4FromVectorAndFloat("Vector (RGB)"=$Output.InterpolateOverTime.Moving Average, "Float (Alpha)"=1)) {isSetParameters=true}
   &SkeletalMeshLocation(
    Bone / Socket Index=$Particles.SpawnGroup
    Bone Sampling Mode=::NewEnumerator4
    MaskSpawnGroup=false
    Orientation Sampling=::NewEnumerator1
    Sampled Orientation Coordinate Space=::Local
    Skeletal Mesh=$User.SK_Mesh
   )
   &ScaleColor(
    Scale Alpha=&FloatFromCurve(
     FloatCurve=DataInterface() {propertyValues {Curve {keys [
         {arriveTangent=1.4675, interpMode=RCIM_Cubic, leaveTangent=1.4675, tangentMode=RCTM_User, time=0, value=0}
         {arriveTangent=-0.8619, interpMode=RCIM_Cubic, leaveTangent=-0.8619, tangentMode=RCTM_User, time=0.4, value=1}
         {arriveTangent=0.8846, interpMode=RCIM_Cubic, leaveTangent=0.8846, tangentMode=RCTM_User, time=1, value=0}
        ]}}}
     Scale Curve=2
    )
    ScaleA=true
    ScaleRGB=true
   )
   &DynamicMaterialParameters(
    Index 0 Param 4=&FloatFromCurve(FloatCurve=DataInterface() {propertyValues {Curve {keys [
         {arriveTangent=1.0462, interpMode=RCIM_Cubic, leaveTangent=1.0462, tangentMode=RCTM_User, time=0, value=-0.7}
         {arriveTangent=0.5091, interpMode=RCIM_Cubic, leaveTangent=0.5091, tangentMode=RCTM_User, time=1, value=1}
        ]}}})
    Index 3 Param 2=&Lerp_Float(A=0, B=100, Alpha=$Particles.Dive)
    Module.Param0WriteEnabled=false
    Module.Param0WriteEnabled3=true
    Module.Param1WriteEnabled=false
    Module.Param1WriteEnabled3=true
    Module.Param2WriteEnabled=false
    Module.Param2WriteEnabled3=true
    Module.Param3WriteEnabled=true
    Module.Param3WriteEnabled3=true
    Write Parameter Index 3=true
   ) {module=DynamicMaterialParameters001}
   &ScaleMeshSize(Scale Factor=&Multiply_VectorByFloat(Vector=<1 1 1>, Float=&Lerp_Float(A=1, B=1.45, Alpha=$Particles.Dive)))
  ]
  rendererClasses [&NiagaraMeshRendererProperties]
  renderers [{rendererClass=&NiagaraMeshRendererProperties, rendererIndex=0}]
 }
 Hand_Mesh {
  emitterUpdateScript [
   &EmitterState(
    Inactive Response=::NewEnumerator1
    Life Cycle Mode=::NewEnumerator1
    Loop Behavior=::NewEnumerator1
    Loop Duration Mode=::NewEnumerator1
    UseLoopDelay=false
   )
   &SpawnBurst_Instantaneous(Spawn Count=2, Use Spawn Probability=false, UseLoopCountLimit=false)
  ]
  particleSpawnScript [
   &InitializeParticle(
    Color Mode=::NewEnumerator0
    Lifetime=5
    Mesh Scale Mode=::NewEnumerator3
    Mesh Uniform Scale=3.5
    UsePositionOffset=false
    Write Material Random=true
   )
   &FN_SetSpawnGroup(Number of Spawn Groups=2)
   &InitialMeshOrientation(Mesh Orientation Mode=::NewEnumerator7, Rotation=<0 0.5 0>, Use Rotation Vector=true) {module=InitialMeshOrientation001}
   &CameraOffset()
  ]
  particleUpdateScript [
   &ParticleState(Kill Particles When Lifetime Has Elapsed=false, Let Infinitely Lived Particles Die When Emitter Deactivates=true)
   &SkeletalMeshLocation(
    Bone / Socket Index=$Particles.SpawnGroup
    Bone Sampling Mode=::NewEnumerator4
    MaskSpawnGroup=false
    Orientation Sampling=::NewEnumerator1
    Sampled Orientation Coordinate Space=::Local
    Sampled Position Offset=<0 0 6>
    Sampled Position Offset Coordinate Space:ENiagaraSkelMeshTransforms=::NewEnumerator1
    Skeletal Mesh=$User.SK_Mesh
   )
   &SolveForcesAndVelocity(Clamp Velocity=false, Limit Acceleration=false)
   &SolveRotationalForcesAndVelocity()
  ]
  rendererClasses [&NiagaraMeshRendererProperties]
  renderers [{rendererClass=&NiagaraMeshRendererProperties, rendererIndex=0}]
 }
 Hand_RIing_Mesh {
  emitterUpdateScript [
   &EmitterState(
    Inactive Response=::NewEnumerator1
    Life Cycle Mode=::NewEnumerator1
    Loop Behavior=::NewEnumerator1
    Loop Duration Mode=::NewEnumerator1
    UseLoopDelay=false
   )
   &SpawnBurst_Instantaneous(Spawn Count=2, Use Spawn Probability=false, UseLoopCountLimit=false)
  ]
  particleSpawnScript [
   &FN_SetSpawnGroup(Number of Spawn Groups=2)
   &InitializeParticle(Color Mode=::NewEnumerator0, Mesh Scale Mode=::NewEnumerator3, Mesh Uniform Scale=0.75, UsePositionOffset=false, Write Material Random=true)
   &InitialMeshOrientation(Mesh Orientation Mode=::NewEnumerator7, Rotation=<0 0 0.25>, Rotation Coordinate Space=::NewEnumerator0, Use Rotation Vector=true)
   &CameraOffset()
  ]
  particleUpdateScript [
   &SkeletalMeshLocation(
    Bone / Socket Index=$Particles.SpawnGroup
    Bone Sampling Mode=::NewEnumerator4
    "Interpolate Mesh Sampling (Update)"=true
    MaskSpawnGroup=false
    Skeletal Mesh=$User.SK_Mesh
   )
   &ParticleState(Kill Particles When Lifetime Has Elapsed=false, Let Infinitely Lived Particles Die When Emitter Deactivates=true)
  ]
  rendererClasses [&NiagaraMeshRendererProperties]
  renderers [{rendererClass=&NiagaraMeshRendererProperties, rendererIndex=0}]
 }
 Light {
  emitterUpdateScript [
   &EmitterState(
    Inactive Response=::NewEnumerator1
    Life Cycle Mode=::NewEnumerator1
    Loop Behavior=::NewEnumerator1
    Loop Duration Mode=::NewEnumerator1
    UseLoopDelay=false
   )
   &SpawnRate(SpawnRate=4, Use Spawn Probability=false) {enabled=false}
   &SpawnBurst_Instantaneous(Spawn Count=2, Use Spawn Probability=false, UseLoopCountLimit=false)
  ]
  particleSpawnScript [
   &FN_SetSpawnGroup(Number of Spawn Groups=2)
   /SetVariables(Particles.LightExponent=2, Particles.LightRadius=40) {isSetParameters=true}
   &InitializeParticle(
    Color=<rgba 0.6462 7.3099 34 1>
    Sprite Size Mode=::NewEnumerator1
    Uniform Sprite Size Max=45
    Uniform Sprite Size Min=40
    UsePositionOffset=false
    Write Material Random=true
   )
  ]
  particleUpdateScript [
   &SkeletalMeshLocation(
    Bone / Socket Index=$Particles.SpawnGroup
    Bone Sampling Mode=::NewEnumerator4
    MaskSpawnGroup=false
    Sampled Position Offset=<0 0 20>
    Sampled Position Offset Coordinate Space:ENiagaraSkelMeshTransforms=::NewEnumerator1
    Skeletal Mesh=$User.SK_Mesh
   )
   &ParticleState(Kill Particles When Lifetime Has Elapsed=false, Let Infinitely Lived Particles Die When Emitter Deactivates=true)
   &ScaleColor(
    Scale RGB=&Multiply_VectorByFloat(Vector=<1 1 1>, Float=&FNTimeOfDayScalar(Night Scale=0.8, Morning Scale=1, Evening Scale=0.8))
    ScaleA=true
    ScaleRGB=true
   ) {module=ScaleColor001}
  ]
  rendererClasses [&NiagaraLightRendererProperties]
  renderers [{rendererClass=&NiagaraLightRendererProperties, rendererIndex=0}]
 }
 Red_Trail {
  emitterUpdateScript [
   &EmitterState(
    Inactive Response=::NewEnumerator1
    Life Cycle Mode=::NewEnumerator1
    Loop Behavior=::NewEnumerator1
    Loop Duration Mode=::NewEnumerator1
    UseLoopDelay=false
   )
   &SpawnBurst_Instantaneous(Spawn Count=4, Use Spawn Probability=false, UseLoopCountLimit=false)
  ]
  particleSpawnScript [
   &FN_SetSpawnGroup()
   &InitializeParticle(
    Color=<rgba 1 1 1 1.2>
    Lifetime=0.5
    Mesh Scale=&SelectVectorFromArray(
     Direct Array Index=&SpawnGroup()
     Vector Selection Array=DataInterface() {propertyValues {FloatData [<0.5 0.5 0.9>, <0.5 0.5 0.9>, <0.26 0.26 0.26>, <0.26 0.26 0.26>]}}
    )
    Mesh Scale Mode=::NewEnumerator4
    UsePositionOffset=false
    Write Material Random=true
   )
   /SetVariables(
    Particles.MeshScale:Vector3f=&SelectVectorFromArray(
     Direct Array Index=&SpawnGroup()
     Vector Selection Array=DataInterface() {propertyValues {FloatData [<0.7 0.7 1.1>, <0.7 0.7 1>, <0.4 0.4 0.3>, <0.4 0.4 0.3>]}}
    )
    Particles.SplineMeshScale:Vector3f=&SelectVectorFromArray(
     Direct Array Index=&SpawnGroup()
     Vector Selection Array=DataInterface() {propertyValues {FloatData [<1 1 11>, <1 1 0.65>, <0.1 0.1 0.14>, <0.1 0.1 0.14>]}}
    )
   ) {isSetParameters=true, module=SetVariables_#1}
   &InitialMeshOrientation(
    Orientation Axis=<0 0 1>
    Orientation Vector=<0 0 1>
    Rotation Coordinate Space=::NewEnumerator0
    Use Orientation Vector=true
    Use Rotation Vector=true
   )
   &CameraOffset(Camera Offset Amount=8)
  ]
  particleUpdateScript [
   &FN_DivingAttributes_v2(
    Accumulate Dive=0.1
    Accumulate Rest=-0.1
    Bone Index=0
    SkeletalMesh=DataInterface() {propertyValues {
      DefaultMesh=null
      ExcludeBoneName=root
      OptionalComponentClass=null
      PreviewMesh=&M_Med_Soldier_04_Preview
      SkinningMode=SkinOnTheFly
      SoftSourceActor=null
      SourceMode=Default
      UvSetIndex=0
      WholeMeshLOD=-1
      bExcludeBone=false
      bReadDeformedGeometry=true
      bRequireCurrentFrameData=true
      ComponentTags []
      FilteredBones [pelvis]
      FilteredSockets []
      MeshUserParameter {parameter {name=None, typeDefHandle {registeredTypeIndex=1}, varData []}}
      SamplingRegions []
     }}
   )
   &ParticleState(Kill Particles When Lifetime Has Elapsed=false, Let Infinitely Lived Particles Die When Emitter Deactivates=true)
   &InterpolateOverTime(Rate Of Change=10, Target Vector Value:NiagaraPosition=$Particles.Position, Variable Type=::NewEnumerator2)
   /SetVariables(Particles.DY:Vector4f=&MakeFloat4FromVectorAndFloat("Vector (RGB)"=$Output.InterpolateOverTime.Moving Average, "Float (Alpha)"=1)) {isSetParameters=true}
   /SetVariables(Particles.DynamicMaterialParameter2:Vector4f=&MakeFloat4FromVectorAndFloat("Vector (RGB)"=$Output.InterpolateOverTime.Moving Average, "Float (Alpha)"=1)) {isSetParameters=true}
   /SetVariables(Particles.Dive01:NiagaraFloat=&Multiply_Float(A=1, B=1)) {isSetParameters=true}
   &DynamicMaterialParameters(
    Index 0 Param 1=&Multiply_Float(
     A=&RandomRangeFloat(Minimum=0.7, Maximum=1.2)
     B=&SelectFloatFromArray(
      Direct Array Index=&SpawnGroup()
      Float Selection Array=DataInterface() {propertyValues {FloatData [1,1,0.5,0.5]}}
     )
    )
    Index 0 Param 2=&Multiply_Float(A=&RandomRangeFloat(Minimum=1, Maximum=1), B=1)
    Index 0 Param 3=&Multiply_Float(
     A=&Multiply_Float(
      A=&RandomRangeFloat(Minimum=1, Maximum=1.8)
      B=&Lerp_Float(A=1, B=1.2, Alpha=$Particles.Dive)
     )
     B=&SelectFloatFromArray(Direct Array Index=$Particles.SpawnGroup, Float Selection Array=DataInterface() {propertyValues {FloatData [1,1,0.3,0.3]}})
    )
    Index 0 Param 4=&RandomRangeFloat(Minimum=0.9, Maximum=2)
    Index 1 Param 1=&RandomRangeFloat(Minimum=1, Maximum=2)
    Index 3 Param 1=&Multiply_Float(
     A=&RandomRangeFloat(Minimum=0.8, Maximum=1.5)
     B=&Lerp_Float(A=1, B=0.5, Alpha=$Particles.Dive)
    )
    Index 3 Param 2=&Multiply_Float(
     A=&SelectFloatFromArray(Direct Array Index=$Particles.SpawnGroup, Float Selection Array=DataInterface() {propertyValues {FloatData [1,1,0,0]}})
     B=&Lerp_Float(A=0, B=90, Alpha=$Particles.Dive)
    )
    Module.Param0WriteEnabled=true
    Module.Param0WriteEnabled1=true
    Module.Param0WriteEnabled3=true
    Module.Param1WriteEnabled=true
    Module.Param1WriteEnabled1=true
    Module.Param1WriteEnabled3=true
    Module.Param2WriteEnabled=true
    Module.Param2WriteEnabled1=true
    Module.Param2WriteEnabled3=true
    Module.Param3WriteEnabled=true
    Module.Param3WriteEnabled1=true
    Module.Param3WriteEnabled3=true
    Write Parameter Index 1=true
    Write Parameter Index 3=true
   )
   &SkeletalMeshLocation(
    Bone / Socket Index=$Particles.SpawnGroup
    Bone Sampling Mode=::NewEnumerator4
    "Interpolate Mesh Sampling (Update)"=true
    MaskSpawnGroup=false
    Orientation Sampling=::NewEnumerator1
    Sampled Orientation Coordinate Space=::Local
    Skeletal Mesh=$User.SK_Mesh
   )
   &ScaleMeshSize(Scale Factor=&Multiply_VectorByFloat(
     Vector=<1 1 1>
     Float=&Multiply_Float(
      A=&FloatFromCurve(FloatCurve=DataInterface() {propertyValues {Curve {keys [<at 0 0>,<at 1 1>]}}})
      B=&Lerp_Float(A=0.82, B=1, Alpha=$Particles.Dive)
     )
    ))
   /SetVariables(Particles.MeshScle1:Vector3f=&Multiply_Vector(A=$Particles.MeshScale, B=&Lerp_Vector(A=<1 1 1>, B=<1 1 1.6>, Alpha=$Particles.Dive))) {isSetParameters=true}
   &ScaleColor(
    Scale Alpha=&FloatFromCurve(FloatCurve=DataInterface() {propertyValues {Curve {keys [<at 0 0>,<at 1 1>]}}})
    ScaleA=true
    ScaleRGB=true
   )
  ]
  rendererClasses [&NiagaraMeshRendererProperties]
  renderers [
   {rendererClass=&NiagaraMeshRendererProperties, rendererIndex=0}
   {rendererClass=&NiagaraMeshRendererProperties, rendererIndex=1}
  ]
 }
}
summary {
 systemName=NS_Contrail_LoneWolf_Blue
 emitters [
  {emitterName=Red_Trail, rendererClasses [&NiagaraMeshRendererProperties]}
  {emitterName=Hand_Mesh, rendererClasses [&NiagaraMeshRendererProperties]}
  {emitterName=Light, rendererClasses [&NiagaraLightRendererProperties]}
  {emitterName=Hand_RIing_Mesh, rendererClasses [&NiagaraMeshRendererProperties]}
  {emitterName=Body_Moon, rendererClasses [&NiagaraMeshRendererProperties]}
  {emitterName=Gloden_Trail, rendererClasses [&NiagaraMeshRendererProperties]}
  {emitterName=Gloden_Trail001, rendererClasses [&NiagaraMeshRendererProperties]}
 ]
 userVariables [
  User.bIsFrontEndPreview {type=NiagaraBool}
  User.Playerspeed {type=NiagaraFloat, defaultValue=0, description=""}
  User.SK_Mesh {
   type=NiagaraDataInterfaceSkeletalMesh
   defaultValue {
    struct=NiagaraExt_VariableValue_DataInterface
    value {dataInterface=&self:NiagaraDataInterfaceSkeletalMesh_0}
   }
  }
  User.Skeletal Mesh01 {
   type=NiagaraDataInterfaceSkeletalMesh
   defaultValue {
    struct=NiagaraExt_VariableValue_DataInterface
    value {dataInterface=&self:NiagaraDataInterfaceSkeletalMesh_1}
   }
  }
 ]
}
